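# Physics asset named for the Target Practice body mesh, a medium male player character body.
# Its constraint list is empty, so it defines no joints between physics bodies.
# No body setups or collision shapes are shown, so it carries no visible ragdoll or simulation setup of its own.
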PhysicsAsset /Game/Characters/Player/Male/Medium/Bodies/M_MED_Target_Practice/Meshes/M_MED_Target_Practice_Physics
constraints []
